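# Physics asset for the Wedding vehicle body cosmetic, giving that mesh a simplified collision shape.
# It holds a single body on the root bone, a box named root_box offset forward and upward from the bone to enclose the vehicle as one volume.
# It has no constraints, so there are no jointed parts.
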PhysicsAsset /VehicleCosmetics/Bodies/Wedding/Meshes/Wedding_PhysicsDM
bodies {root [{shapeType=Box, extentX=456, extentY=216, extentZ=132, shapeName=root_box, center=<24.2949 0 103.7205>}]}
constraints []
